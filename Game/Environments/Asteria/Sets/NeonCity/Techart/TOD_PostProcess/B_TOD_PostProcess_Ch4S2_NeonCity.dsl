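; Neon City post-process actor for Chapter 4 Season 2, driving two post-process volumes (standard and Lumen) plus volumetric fog within a sized sphere.
; On cosmetic-capable instances, UpdatePostVolumes runs at construction, BeginPlay and scalability changes.
; It gates components by quality settings, Lumen support and, for standard volume and fog, non-mobile/Switch platforms.
; Epic Lumen blocking-loads a high-res FFT bloom kernel, otherwise a low-res one, and debug events toggle or log state.
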(event Custom|LoadLumenEpicBloomKernel
  ;[1 1 1] Load / clear high res convolution kernel
  (bind _returnvalue (Utilities|LoadAssetBlocking "/Game/TimeOfDay/TODM_Sequence/Textures/Bloom/T_FFTBloom_Asteria0000.T_FFTBloom_Asteria0000"))
  (bind _astexture_2d (Utilities|Casting|CastToTexture2D _returnvalue))
  (bind _structout (Utilities|Struct|SetmembersinPostProcessSettings (Variables|/Script/Engine.PostProcessComponent|GetSettings (Variables|Default|GetPP_Vol_Lumen)) "BM_FFT" _astexture_2d))
  (Variables|/Script/Engine.PostProcessComponent|SetSettings _structout (Variables|Default|GetPP_Vol_Lumen)))

(event EventBeginPlay
  (|CallParentFunction)
  (Networking|CanExecuteCosmeticEvents
    (:True
      (CallFunction|BindScalabilityEvents)
      (CallFunction|UpdatePostVolumes))))

(event Custom|LoadLumenHighBloomKernel)

(event Custom|EnableNeonCityPostProcess
  (CallFunction|UpdatePostVolumes))

(event Custom|ClearLumenBloomKernel
  ;[1 1 1] Load / clear high res convolution kernel
  (bind _structout (Utilities|Struct|SetmembersinPostProcessSettings (Variables|/Script/Engine.PostProcessComponent|GetSettings (Variables|Default|GetPP_Vol_Lumen)) "/Game/TimeOfDay/TODM_Sequence/Textures/Bloom/T_lowres_bloom_kernal.T_lowres_bloom_kernal"))
  (Variables|/Script/Engine.PostProcessComponent|SetSettings _structout (Variables|Default|GetPP_Vol_Lumen)))

(event Custom|DisableNeonCityPostProcess
  ;[1 1 1] Debug request from capture team
  (Variables|/Script/Engine.PostProcessComponent|SetbEnabled false (Variables|Default|GetPP_Vol_Lumen))
  (Variables|/Script/Engine.PostProcessComponent|SetbEnabled false (Variables|Default|GetPP_Vol))
  (Rendering|SetVisibility (Variables|Default|GetVolumetricFogVolume) false false))

(event Custom|TogglePPVolumeLumen
  ;[1 1 1] Editor debug utilities
  (Variables|/Script/Engine.PostProcessComponent|SetbEnabled (not (Variables|/Script/Engine.PostProcessComponent|GetbEnabled (Variables|Default|GetPP_Vol_Lumen))) (Variables|Default|GetPP_Vol_Lumen)))

(event Custom|TogglePPVolume
  ;[1 1 1] Editor debug utilities
  (Variables|/Script/Engine.PostProcessComponent|SetbEnabled (not (Variables|/Script/Engine.PostProcessComponent|GetbEnabled (Variables|Default|GetPP_Vol))) (Variables|Default|GetPP_Vol)))

(event Custom|ToggleVolumetricFog
  ;[1 1 1] Editor debug utilities
  (Rendering|SetVisibility (Variables|Default|GetVolumetricFogVolume) (not (Variables|/Script/Engine.SceneComponent|GetbVisible (Variables|Default|GetVolumetricFogVolume))) false))

(fn BindScalabilityEvents
  (Default|BindScalabilitySettings (EventDispatchers|CreateEvent :!event UpdatePostVolumes)))

(fn DebugNeonCityScalabilitySettings
  (bind _returnvalue (Settings|GetGameUserSettings))
  (Fortnite|FortShippingLog 0 (Utilities|String|Append "Neon City PP Debug Info: - " "Overall Scalability Level =" (Utilities|String|ToString(Integer) (Settings|GetOverallScalabilityLevel _returnvalue)) " - Post Procession Quality = " (Utilities|String|ToString(Integer) (Settings|GetPostProcessingQuality _returnvalue)) " - Is Lumen Enabled = " (Utilities|String|ToString(Boolean) (Class|BTechArtGlobalLibrary|IsLumenEnabled)) " - Is Mobile Game = " (Utilities|String|ToString(Boolean) (TouchPlatform|IsMobileGame)) " - Is Erebus = " (Utilities|String|ToString(Boolean) (TouchPlatform|IsNintendoSwitch)) " - PP Vol Enabled = " (Utilities|String|ToString(Boolean) (Variables|/Script/Engine.PostProcessComponent|GetbEnabled (Variables|Default|GetPP_Vol))) " - PP Vol Lumen Enabled = " (Utilities|String|ToString(Boolean) (Variables|/Script/Engine.PostProcessComponent|GetbEnabled (Variables|Default|GetPP_Vol_Lumen))) " - Volumetric Fog Volume Visible = " (Utilities|String|ToString(Boolean) (Variables|/Script/Engine.SceneComponent|GetbVisible (Variables|Default|GetVolumetricFogVolume))))))

(fn SetPostBlendRadius
  (Variables|/Script/Engine.PostProcessComponent|SetBlendRadius (Variables|Default|GetPP%20Blend%20Radius) (Variables|Default|GetPP_Vol_Lumen))
  (Variables|/Script/Engine.PostProcessComponent|SetBlendRadius (Variables|Default|GetPP%20Blend%20Radius) (Variables|Default|GetPP_Vol)))

(fn UpdatePostVolumes
  (bind _returnvalue (TouchPlatform|IsNintendoSwitch))
  (bind _returnvalue_2 (TouchPlatform|IsMobileGame))
  (bind _returnvalue_3 (or _returnvalue_2 _returnvalue))
  (bind _returnvalue_4 (not _returnvalue_3))
  (bind _returnvalue_5 (Settings|GetGameUserSettings))
  (LocalVariables|SetPostProcessingQuality (Settings|GetPostProcessingQuality _returnvalue_5))
  (LocalVariables|SetGlobalIlluminationQuality (Settings|GetGlobalIlluminationQuality _returnvalue_5))
  (LocalVariables|SetShadowQuality (Settings|GetShadowQuality _returnvalue_5))
  ;[1 1 1] Check to enable Lumen-specific PP
  (bind _output_get (Variables|/Script/Engine.PostProcessComponent|SetbEnabled (and (Class|BTechArtGlobalLibrary|IsLumenEnabled) (>= (LocalVariables|GetPostProcessingQuality) 2)) (Variables|Default|GetPP_Vol_Lumen)))
  (Variables|Default|SetIsLumenHighPlatform _output_get)
  (Variables|Default|SetIsLumenEpicPlatform (and (>= (LocalVariables|GetPostProcessingQuality) 3) (Class|BTechArtGlobalLibrary|IsLumenEnabled)))
  ;[1 1 1] Check which PP should be enabled
  (Variables|/Script/Engine.PostProcessComponent|SetbEnabled (Math|Boolean|ANDBoolean (>= (LocalVariables|GetGlobalIlluminationQuality) 2) (>= (LocalVariables|GetPostProcessingQuality) 2) _returnvalue_4) (Variables|Default|GetPP_Vol))
  (Rendering|SetVisibility (Variables|Default|GetVolumetricFogVolume) (and (>= (LocalVariables|GetShadowQuality) 2) _returnvalue_4) false)
  (if (Variables|Default|GetIsLumenEpicPlatform)
    (CallFunction|LoadLumenEpicBloomKernel)
    (else
      (CallFunction|ClearLumenBloomKernel))))

(fn ConstructionScript
  (bind _spherevolumeradius (Variables|Default|GetSphereVolumeRadius))
  (|CallParentFunction)
  (Networking|CanExecuteCosmeticEvents
    (:True
      (Utilities|FlowControl|Sequence
        (:then_0
          (Components|Sphere|SetSphereRadius (Variables|Default|GetSphere) _spherevolumeradius true)
          (Transformation|SetRelativeScale3D (Variables|Default|GetVolumetricFogVolume) (Math|Conversions|ToVector(Float) _spherevolumeradius)))
        (:then_1
          (CallFunction|SetPostBlendRadius)
          (CallFunction|UpdatePostVolumes))))))
