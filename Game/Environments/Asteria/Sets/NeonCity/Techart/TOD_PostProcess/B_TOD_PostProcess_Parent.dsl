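; Parent Blueprint for time-of-day post process in the NeonCity environment set, providing a shared day phase lookup.
; GetDayPhaseScalar breaks the current active time-of-day color into R, G, B and A as Morning, Day, Evening and Night weights.
; It selects the channel for the requested phase and returns whether it is above zero, plus the value itself.
; BeginPlay, Tick and the construction script are empty.
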(event EventBeginPlay)

(event EventTick
  (params (DeltaSeconds :type Float%20%28single-precision%29)))

(fn GetDayPhaseScalar
  (params (Day%20Phase :type EFortDayPhase%20Enum :default "Morning"))
  (bind (_currentactivetimeofday_r _currentactivetimeofday_g _currentactivetimeofday_b _currentactivetimeofday_a) (Math|Color|BreakColor (Variables|GetCurrentActiveTimeOfDay_R)))
  (return (> (Utilities|Select :Index Day%20Phase :Morning _currentactivetimeofday_r :Day _currentactivetimeofday_g :Evening _currentactivetimeofday_b :Night _currentactivetimeofday_a) 0) (Utilities|Select :Index Day%20Phase :Morning _currentactivetimeofday_r :Day _currentactivetimeofday_g :Evening _currentactivetimeofday_b :Night _currentactivetimeofday_a)))

(fn ConstructionScript)
