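# Additive volume-domain material for NeonCity local fog volumes, with extinction set per time of day through Scalar_Time_of_Day.
# Two world-space curl noise samples modulate extinction, replaced by a constant on low/medium quality and forward/mobile paths.
# A "PPV Fade" reroute derives from camera-to-object distance, apparently fading extinction out as the camera moves away.
# It also defines an Albedo tint and an actor-centred sphere mask sized by scaled bounds X.
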Material /Game/Environments/Asteria/Sets/NeonCity/Techart/TOD_PostProcess/M_TOD_FogVolume
graph {
 properties {blendMode=BLEND_Additive, materialDomain=MD_Volume, pixelDepthOffsetMode=PDOM_Legacy, usedShadingModels=Default Lit}
 nodes [
  n1 = ScalarParameter() {defaultValue=0.0002, parameterName=Extinction_Morning}
  n2 = ScalarParameter() {parameterName=Extinction_Day}
  n3 = ScalarParameter() {defaultValue=0.0002, parameterName=Extinction_Evening}
  n4 = ScalarParameter() {defaultValue=0.0003, parameterName=Extinction_Night}
  n5 = TextureSample(UVs=Divide(A=WorldPosition().XYZ, B=ScalarParameter() {defaultValue=17500, parameterName=CurlA_WorldDivisor})) {samplerType=SAMPLERTYPE_LinearColor, texture=/Game/Effects/FogVolume/VT_CurlNoise_LOW}
  n6 = ScalarParameter() {defaultValue=1.8, parameterName=CurlA_Bias}
  n7 = TextureSample(UVs=Divide(A=WorldPosition().XYZ, B=ScalarParameter() {defaultValue=22500, parameterName=CurlB_WorldDivisor})) {samplerType=SAMPLERTYPE_LinearColor, texture=&VT_CurlNoise_LOW}
  n8 = ScalarParameter() {defaultValue=2.5, parameterName=CurlB_Bias}
  n9 = Constant() {r=1}
  n10 = QualitySwitch(Default=Add(A=Add(A=@n5.R, B=@n6), B=Add(A=@n7.G, B=@n8)), Low=@n9, Medium=@n9)
  Multiply(
   A=Multiply(
    A=MaterialFunctionCall(Value_Morning=@n1, Value_Day=@n2, Value_Evening=@n3, Value_Night=@n4) {materialFunction=&Scalar_Time_of_Day}.Result
    B=ShadingPathSwitch(Default=@n10, Forward=@n9, Mobile=@n9)
   )
   B=OneMinus(None=NamedRerouteUsage())
  )
  n11 = Subtract(A=Distance(A=ObjectPositionWS(), B=CameraPositionWS()), B=ScalarParameter() {defaultValue=15000, parameterName=PPV_Transition_Offset})
  n12 = ScalarParameter() {defaultValue=2000, parameterName=PPV_Transition_Length}
  NamedRerouteDeclaration(None=Saturate(None=Divide(A=@n11, B=@n12))) {name=PPV Fade, nodeColor=<rgba 1 0 0.7266 1>}
  n13 = MaterialFunctionCall(Float3=Multiply(
    A=MaterialFunctionCall() {materialFunction=&ObjectLocalBounds}.Local Bounds Max
    B=MaterialFunctionCall() {materialFunction=&ObjectScale}.Scale XYZ
   )) {materialFunction=&BreakOutFloat3Components}
  SphereMask(
   A=ActorPositionWS()
   B=WorldPosition().XYZ
   Radius=Max(A=@n13.R, B=@n13.R)
   Hardness=ScalarParameter() {defaultValue=0.7, parameterName=Hardness}
  )
  VectorParameter() {parameterName=Albedo, defaultValue=<rgba 0.3079 0.5841 0.6267 0>}
 ]
}
